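; Season 22 signal booster quest prop whose radar dish turns to face the tree when a player with the quest damages it.
; Server overlap and damage events check the instigator's quest objective, then trigger weakpoint VFX, red or green emissive, rotation sound and a timeline-driven yaw spin.
; Dish and VFX changes run only on the instigator's client, with splitscreen handled.
; A test flag bypasses the quest check and resets visuals after a cooldown.
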(event OnComponentBeginOverlap(BoundingBox)
  (params (OverlappedComponent :type Primitive%20Component%20Object%20Reference) (OtherActor :type Actor%20Object%20Reference) (OtherComp :type Primitive%20Component%20Object%20Reference) (OtherBodyIndex :type Integer) (bFromSweep :type Boolean) (SweepResult :type Hit%20Result%20Structure%20%28by%20ref%29))
  ;[0 0 0] On begin or end overlap - toggle/remove weakpoint VFX to show the prop is interactable if the player has the quest
  (Networking|SwitchHasAuthority
    (:Authority
      (bind _asfort_player_pawn (Utilities|Casting|CastToFortPlayerPawn OtherActor))
      (bind _bcaninteract (CallFunction|InstigatorHasQuest _asfort_player_pawn))
      (if _bcaninteract
        (CallFunction|ToggleRadarVisuals _asfort_player_pawn false true)))))

(event OnComponentEndOverlap(BoundingBox)
  (params (OverlappedComponent :type Primitive%20Component%20Object%20Reference) (OtherActor :type Actor%20Object%20Reference) (OtherComp :type Primitive%20Component%20Object%20Reference) (OtherBodyIndex :type Integer))
  ;[0 0 0] On begin or end overlap - toggle/remove weakpoint VFX to show the prop is interactable if the player has the quest
  (Networking|SwitchHasAuthority
    (:Authority
      (bind _asfort_player_pawn (Utilities|Casting|CastToFortPlayerPawn OtherActor))
      (bind _bcaninteract (CallFunction|InstigatorHasQuest _asfort_player_pawn))
      (if _bcaninteract
        (CallFunction|ToggleRadarVisuals _asfort_player_pawn)))))

(event Custom|SpinDishToTree
  (params (InstigatingPlayerPawn :type Fort%20Player%20Pawn%20Object%20Reference) (StartRotation :type Rotator) (EndRotation :type Rotator))
  ;[0 0 0] Client only - spins the dish so it faces the tree for the instigating player
  (Networking|SwitchHasAuthority
    (:Remote
      (bind _output (CallFunction|IsValidLocalPlayerController (Fortnite|GetFortPlayerControllerfromActor InstigatingPlayerPawn)))
      (if _output
        (bind (_direction _lerpamount) (Timeline|Timeline PlayFromStart)
          (:Update
            (Transformation|SetRelativeRotation (Variables|Default|GetSM_Dish) (Math|Rotator|MakeRotator 0.0 0.0 (Math|Float|Lerp (.yaw StartRotation) (.yaw EndRotation) _lerpamount)) false false)))))))

(event Custom|ToggleRadarVisuals
  (params (InstigatingPlayerPawn :type Fort%20Player%20Pawn%20Object%20Reference) (bHasCompletedQuest :type Boolean) (bShouldShowWeakpoint :type Boolean))
  ;[0 0 0] Client only - toggles radar visuals for the player depending on state (did they leave/enter the volume, did they complete the quest)
  (Networking|SwitchHasAuthority
    (:Remote
      (bind _output (CallFunction|IsValidLocalPlayerController (Fortnite|GetFortPlayerControllerfromActor InstigatingPlayerPawn)))
      (if _output
        (if bHasCompletedQuest
          (Components|Activation|Deactivate (Variables|Default|GetNS_WeakPointEffect_01))
          (Components|Activation|Activate (Variables|Default|GetNS_HitCrackEffect_01) false)
          (Rendering|Material|SetVectorParameterValueonMaterials (Variables|Default|GetSM_Dish) "Emissive" (Variables|Default|GetEmissiveGreen))
          (CallFunction|PlayRotationAudio)
          (elif bShouldShowWeakpoint
            (Components|Activation|Activate (Variables|Default|GetNS_WeakPointEffect_01) false)
            (Rendering|Material|SetVectorParameterValueonMaterials (Variables|Default|GetSM_Dish) "Emissive" (Variables|Default|GetEmissiveRed))
            (else
              (Components|Activation|Deactivate (Variables|Default|GetNS_WeakPointEffect_01))
              (Components|Activation|Activate (Variables|Default|GetNS_HitCrackEffect_01) false))))))))

(event Custom|TestOnlyResetPropVisuals
  (params (InstigatingPlayerPawn :type Fort%20Player%20Pawn%20Object%20Reference))
  ;[0 0 0] Test mode only - resets visuals for the interacting client after the duration has passed
  (Networking|SwitchHasAuthority
    (:Remote
      (Utilities|FlowControl|Delay (Variables|Design|QuestSetup|GetTestingCooldown))
      (bind _output (CallFunction|IsValidLocalPlayerController (Fortnite|GetFortPlayerControllerfromActor InstigatingPlayerPawn)))
      (if _output
        (Transformation|SetRelativeRotation (Variables|Default|GetSM_Dish) "0, 0, 0" false false)
        (Rendering|Material|SetVectorParameterValueonMaterials (Variables|Default|GetSM_Dish) "Emissive" (Variables|Default|GetEmissiveRed))
        (Components|Activation|Activate (Variables|Default|GetNS_WeakPointEffect_01) false)))))

(event EventOnDamageServer
  (params (Damage :type Float%20%28single-precision%29) (DamageTags :type Gameplay%20Tag%20Container%20Structure%20%28by%20ref%29) (Momentum :type Vector :default "0, 0, 0") (HitInfo :type Hit%20Result%20Structure%20%28by%20ref%29) (InstigatedBy :type Controller%20Object%20Reference) (DamageCauser :type Actor%20Object%20Reference) (EffectContext :type Gameplay%20Effect%20Context%20Handle%20Structure))
  ;[0 0 0] On damage server - calls parent logic, then checks if the player has the quest. Changes visual state, plays audio if applicable
  (|CallParentFunction Damage DamageTags Momentum HitInfo InstigatedBy DamageCauser EffectContext)
  (bind _asfort_player_pawn (Utilities|Casting|CastToFortPlayerPawn (Pawn|GetControlledPawn InstigatedBy)))
  (bind _bcaninteract (CallFunction|InstigatorHasQuest _asfort_player_pawn))
  (if _bcaninteract
    (bind _rotation (CallFunction|CalculateRotationForRadar))
    (CallFunction|SpinDishtoTree _asfort_player_pawn (Transformation|GetWorldRotation (Variables|Default|GetSM_Dish)) _rotation)
    (CallFunction|ToggleRadarVisuals _asfort_player_pawn true)
    (if (Variables|Design|QuestSetup|GetbTestVisualsNoQuestRequired)
      (CallFunction|TestOnlyResetPropVisuals _asfort_player_pawn))))

(event Custom|PlayRotationAudio
  ;[0 0 0] Client only - plays the rotation audio. Should only be called when the prop enters a successful interact state
  (Networking|SwitchHasAuthority
    (:Remote
      (Development|PrintString "rotation audio playing" true true "(R=1.000000,G=0.000000,B=0.390266,A=1.000000)")
      (Audio|PlaySoundatLocation (Variables|Audio|GetRotationSound) (Transformation|GetWorldLocation (Variables|Default|GetSM_Base))))))

(fn CalculateRotationforRadar
  ;[0 0 0] Calculate the angle  that the radar will need to turn to face the tree. The way the dish faces forward is actually the right vector, so append 90 deg to the final rotation (assumes start rotation of actor is 0!)
  (return (Math|Rotator|MakeRotator 0.0 0.0 (+ (.yaw (Math|Rotator|FindLookatRotation (Variables|Default|GetTreeWorldLocation) (Transformation|GetActorLocation))) 9e+01))))

(fn InstigatorHasQuest
  (params (InstigatingPlayerPawn :type Fort%20Player%20Pawn%20Object%20Reference))
  ;[0 0 0] Check if the instigating player has the associated quest OR if test mode is enabled
  (Networking|SwitchHasAuthority
    (:Authority
      (if (Variables|Design|QuestSetup|GetbTestVisualsNoQuestRequired)
        (return true)
        (else
          ;;[0.25 0.25 0.25] Ensure that prop's associated quest objective is not completed and that it is also in progress
          (return (and (not (Quest|HasCompletedObjectivewithName (Quests|GetQuestManager (Fortnite|GetFortPlayerControllerfromActor InstigatingPlayerPawn) "Invalid") (Class|ConsolidatedQuestComponent|GetQuestItemDefinition (Variables|BuildingGameplayActorPropQuest|GetConsolidatedQuestComponent)) (Variables|/Script/FortniteGame.ConsolidatedQuestComponent|GetObjectiveBackendName (Variables|BuildingGameplayActorPropQuest|GetConsolidatedQuestComponent)) :!class /Script/FortniteGame.FortQuestManager)) (Quest|IsObjectivewithNameinProgress (Quests|GetQuestManager (Fortnite|GetFortPlayerControllerfromActor InstigatingPlayerPawn) "Invalid") (Class|ConsolidatedQuestComponent|GetQuestItemDefinition (Variables|BuildingGameplayActorPropQuest|GetConsolidatedQuestComponent)) (Variables|/Script/FortniteGame.ConsolidatedQuestComponent|GetObjectiveBackendName (Variables|BuildingGameplayActorPropQuest|GetConsolidatedQuestComponent)) :!class /Script/FortniteGame.FortQuestManager))))))))

(fn IsValidLocalPlayerController
  (params (InstigatingPlayerController :type Fort%20Player%20Controller%20Object%20Reference))
  ;[0 0 0] Returns true on the executing client if the instigator controller is locally controlled. Returns false if running on the server
  (if (not (Networking|IsServer))
    (Utilities|IsValid InstigatingPlayerController
      (:Is%20Valid
        (if (PlayerController|IsinSplitscreen InstigatingPlayerController)
          (for _array_element (Fortnite|GetLocalFortPlayerControllers true)
            (if (== _array_element InstigatingPlayerController)
              (return true)))
          (return false)
          (else
            (return true))))
      (:Is%20Not%20Valid
        _)))
  (return false))

(fn ConstructionScript
  (|CallParentFunction))
