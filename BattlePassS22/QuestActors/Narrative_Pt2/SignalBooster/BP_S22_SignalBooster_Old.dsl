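; Season 22 Signal Booster quest prop: a radar dish that players on the quest damage while it is correctly aligned.
; Server overlap sends the pawn a gameplay event carrying this actor and cancels the quest ability on exit.
; Hits without the Valid tag apply an invalid-hit effect and sound; valid hits or completed interaction cancel the ability.
; Dish rotation, emissive colour, weakpoint FX and audio run client-side only for the instigating local player, including split-screen.
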(event OnComponentBeginOverlap(BoundingBox)
  (params (OverlappedComponent :type Primitive%20Component%20Object%20Reference) (OtherActor :type Actor%20Object%20Reference) (OtherComp :type Primitive%20Component%20Object%20Reference) (OtherBodyIndex :type Integer) (bFromSweep :type Boolean) (SweepResult :type Hit%20Result%20Structure%20%28by%20ref%29))
  ;[0 0 0] Logic for when the dish is approached, specifically when any actor enters the bounding box
  ;;[0.1927 0.1927 0.1927] Check if the actor entering this bounding box is a player with the associated quest
  (Networking|SwitchHasAuthority
    (:Authority
      (bind _asfort_player_pawn (Utilities|Casting|CastToFortPlayerPawn OtherActor))
      (bind _bcaninteract (CallFunction|InstigatorHasQuest _asfort_player_pawn))
      (if _bcaninteract
        ;;[0.1927 0.1927 0.1927] Send a gameplay event that also sends over this actor instance reference in the payload
        (Ability|SendGameplayEventtoActor _asfort_player_pawn "(TagName=\"Event.Quest.SignalBooster\")" (Utilities|Struct|MakeGameplayEventData "(TagName=\"\")" _asfort_player_pawn self))
        (CallFunction|ToggleRotationSound _asfort_player_pawn)))))

(event OnComponentEndOverlap(BoundingBox)
  (params (OverlappedComponent :type Primitive%20Component%20Object%20Reference) (OtherActor :type Actor%20Object%20Reference) (OtherComp :type Primitive%20Component%20Object%20Reference) (OtherBodyIndex :type Integer))
  ;[0 0 0] Logic for when the dish is approached, specifically when any actor enters the bounding box
  ;;[0.1927 0.1927 0.1927] Check if the actor leaving this bounding box is a player with the associated quest
  (Networking|SwitchHasAuthority
    (:Authority
      (bind _asfort_player_pawn (Utilities|Casting|CastToFortPlayerPawn OtherActor))
      (bind _bcaninteract (CallFunction|InstigatorHasQuest _asfort_player_pawn))
      (if _bcaninteract
        ;;[0.1927 0.1927 0.1927] Cancel the quest ability associated with this quest for the instigator
        (Tags|CancelAbilitiesWithTags (Ability|GetFortAbilitySystemComponent _asfort_player_pawn) (Variables|Default|GetQuestAbilityTagContainer))
        (CallFunction|ToggleRotationSound _asfort_player_pawn true)))))

(event EventOnDamageServer
  (params (Damage :type Float%20%28single-precision%29) (DamageTags :type Gameplay%20Tag%20Container%20Structure%20%28by%20ref%29) (Momentum :type Vector :default "0, 0, 0") (HitInfo :type Hit%20Result%20Structure%20%28by%20ref%29) (InstigatedBy :type Controller%20Object%20Reference) (DamageCauser :type Actor%20Object%20Reference) (EffectContext :type Gameplay%20Effect%20Context%20Handle%20Structure))
  ;[0 0 0] OnDamageServer - if the radar is not facing the tree when hit, play an incorrect sound and pause the rotation for a short duration
  (|CallParentFunction Damage DamageTags Momentum HitInfo InstigatedBy DamageCauser EffectContext)
  (bind _asfort_player_pawn (Utilities|Casting|CastToFortPlayerPawn (Pawn|GetControlledPawn InstigatedBy)))
  (bind _bcaninteract (CallFunction|InstigatorHasQuest _asfort_player_pawn))
  (if _bcaninteract
    (if (not (GameplayTags|HasMatchingGameplayTag _asfort_player_pawn "(TagName=\"Granted.Athena.Quest.S22.SignalBooster.Valid\")"))
      (GameplayEffects|ApplyGameplayEffectToSelf :self (Ability|GetFortAbilitySystemComponent _asfort_player_pawn) :GameplayEffectClass "/BattlePassS22/QuestActors/Narrative_Pt2/SignalBooster/GE_SignalBooster_GrantInvalidHitTag.GE_SignalBooster_GrantInvalidHitTag_C" :Level 0.0)
      (CallFunction|PlaySoundforInstigator _asfort_player_pawn (Variables|Audio|GetInvalidHitSound))
      (else
        (Tags|CancelAbilitiesWithTags (Ability|GetFortAbilitySystemComponent _asfort_player_pawn) (Variables|Default|GetQuestAbilityTagContainer))))))

(event Custom|Lerp%20Dish%20Rotation
  (params (InstigatingPlayerPawn :type Fort%20Player%20Pawn%20Object%20Reference) (CurrentRotationYaw :type Float%20%28double-precision%29) (PredictedRotationYaw :type Float%20%28double-precision%29))
  ;[0 0 0] Cosmetic only - Use timeline to smoothly rotate the dish between two given rotational yaws for the instigating player
  (Networking|SwitchHasAuthority
    (:Remote
      (bind _output (CallFunction|IsValidLocalPlayerController (Fortnite|GetFortPlayerControllerfromActor InstigatingPlayerPawn)))
      (if _output
        (bind (_direction _lerpamount) (Timeline|Timeline PlayFromStart)
          (:Update
            (Transformation|SetRelativeRotation (Variables|Default|GetSM_Dish) (Math|Rotator|MakeRotator 0.0 0.0 (Math|Float|Lerp CurrentRotationYaw PredictedRotationYaw _lerpamount)) false false)))))))

(event Custom|Toggle%20Cosmetic%20On%20Radar
  (params (InstigatingPlayerPawn :type Fort%20Player%20Pawn%20Object%20Reference) (bIsInValidRange :type Boolean))
  (bind _sm_dish (Variables|Default|GetSM_Dish))
  (bind _returnvalue (Components|GetChildComponent _sm_dish 0))
  (bind _returnvalue_2 (Components|GetChildComponent _sm_dish 1))
  ;[0 0 0] Cosmetic only - if the radar is facing the tree, change material parameter for light and toggle weakpoint FX
  (Networking|SwitchHasAuthority
    (:Remote
      (bind _output (CallFunction|IsValidLocalPlayerController (Fortnite|GetFortPlayerControllerfromActor InstigatingPlayerPawn)))
      (if _output
        (if bIsInValidRange
          (Rendering|Material|SetVectorParameterValueonMaterials (Variables|Default|GetSM_Dish) "Emissive" (Variables|Default|GetEmissiveGreen))
          (Components|Activation|Activate _returnvalue false)
          (Components|Activation|Deactivate _returnvalue_2)
          (Variables|Default|SetbEndParticleShouldActivate true)
          (else
            (Rendering|Material|SetVectorParameterValueonMaterials (Variables|Default|GetSM_Dish) "Emissive" (Variables|Default|GetEmissiveRed))
            (Components|Activation|Deactivate _returnvalue)
            (if (Variables|Default|GetbEndParticleShouldActivate)
              (Components|Activation|Activate _returnvalue_2 false)
              (Variables|Default|SetbEndParticleShouldActivate false))))))))

(event Custom|Play%20Sound%20For%20Instigator
  (params (InstigatingPlayerPawn :type Fort%20Player%20Pawn%20Object%20Reference) (CueToPlay :type Sound%20Cue%20Object%20Reference))
  ;[0 0 0] Audio - Client Only
  ;;[0.1146 0.1146 0.1146] Plays a one-time audio cue for the instigating player
  (bind _output (CallFunction|IsValidLocalPlayerController (Fortnite|GetFortPlayerControllerfromActor InstigatingPlayerPawn)))
  (if _output
    (Audio|PlaySoundatLocation CueToPlay (Transformation|GetWorldLocation (Variables|Default|GetSM_Base)))))

(event Custom|ToggleRotationSound
  (params (InstigatingPlayerPawn :type Fort%20Player%20Pawn%20Object%20Reference) (bShouldStop :type Boolean))
  (bind _ambientaudio (Variables|Components|GetAmbientAudio))
  (bind _returnvalue (Audio|Components|Audio|GetPlayState _ambientaudio))
  (bind _returnvalue_2 (Utilities|Enum|Equal(Enum) _returnvalue "Playing"))
  ;[0 0 0] Audio - Client Only
  ;;[0.1823 0.1823 0.1823] Handles toggling the looping sound of the radar dish
  (Networking|SwitchHasAuthority
    (:Remote
      (bind _output (CallFunction|IsValidLocalPlayerController (Fortnite|GetFortPlayerControllerfromActor InstigatingPlayerPawn)))
      (if _output
        (if bShouldStop
          (if _returnvalue_2
            (Audio|Components|Audio|FadeOut (Variables|Components|GetAmbientAudio) 0.5 0.0 "Linear" :!class /Script/Engine.AudioComponent))
          (else
            (if (not _returnvalue_2)
              (Audio|Components|Audio|FadeIn _ambientaudio 0.5 1.0 0.0 "Linear" :!class /Script/Engine.AudioComponent))))))))

(fn InstigatorHasQuest
  (params (InstigatingPlayerPawn :type Fort%20Player%20Pawn%20Object%20Reference))
  ;[0 0 0] Check if the instigating player has the associated quest OR if test mode is enabled
  (Networking|SwitchHasAuthority
    (:Authority
      (if (Variables|Design|QuestSetup|GetbTestVisualsNoQuestRequired)
        (return true)
        (else
          ;;[0.25 0.25 0.25] Ensure that prop's associated quest objective is not completed and that it is also in progress
          (return (and (not (Quest|HasCompletedObjectivewithName (Quests|GetQuestManager (Fortnite|GetFortPlayerControllerfromActor InstigatingPlayerPawn) "Invalid") (Class|ConsolidatedQuestComponent|GetQuestItemDefinition (Variables|BuildingGameplayActorPropQuest|GetConsolidatedQuestComponent)) (Variables|/Script/FortniteGame.ConsolidatedQuestComponent|GetObjectiveBackendName (Variables|BuildingGameplayActorPropQuest|GetConsolidatedQuestComponent)) :!class /Script/FortniteGame.FortQuestManager)) (Quest|IsObjectivewithNameinProgress (Quests|GetQuestManager (Fortnite|GetFortPlayerControllerfromActor InstigatingPlayerPawn) "Invalid") (Class|ConsolidatedQuestComponent|GetQuestItemDefinition (Variables|BuildingGameplayActorPropQuest|GetConsolidatedQuestComponent)) (Variables|/Script/FortniteGame.ConsolidatedQuestComponent|GetObjectiveBackendName (Variables|BuildingGameplayActorPropQuest|GetConsolidatedQuestComponent)) :!class /Script/FortniteGame.FortQuestManager))))))))

(fn IsValidLocalPlayerController
  (params (InstigatingPlayerController :type Fort%20Player%20Controller%20Object%20Reference))
  ;[0 0 0] Returns true on the executing client if the instigator controller is locally controlled. Returns false if running on the server
  (if (not (Networking|IsServer))
    (Utilities|IsValid InstigatingPlayerController
      (:Is%20Valid
        (if (PlayerController|IsinSplitscreen InstigatingPlayerController)
          (for _array_element (Fortnite|GetLocalFortPlayerControllers true)
            (if (== _array_element InstigatingPlayerController)
              (return true)))
          (return false)
          (else
            (return true))))
      (:Is%20Not%20Valid
        _)))
  (return false))

(fn PostInteractionVFX&Audio
  (params (PlayerController :type Fort%20Player%20Controller%20Object%20Reference))
  (bind _sm_dish (Variables|Default|GetSM_Dish))
  ;[0 0 0] When radar has been successfully interacted with, end the quest ability on the player and reset its cosmetics
  (|CallParentFunction PlayerController)
  (if (Networking|IsServer)
    (Tags|CancelAbilitiesWithTags (Ability|GetFortAbilitySystemComponent (Game|Player|GetPlayerPawn PlayerController)) (Variables|Default|GetQuestAbilityTagContainer))
    (else
      (bind _output (CallFunction|IsValidLocalPlayerController PlayerController))
      (if _output
        (Rendering|Material|SetVectorParameterValueonMaterials _sm_dish "Emissive" (Variables|Default|GetEmissiveGreen))
        (Components|Activation|Deactivate (Components|GetChildComponent _sm_dish 0))
        (Components|Activation|Activate (Components|GetChildComponent (Variables|Default|GetSM_Dish) 1) false)))))

(fn ConstructionScript
  (|CallParentFunction)
  (Audio|Components|Audio|SetSound (Variables|Components|GetAmbientAudio) (Variables|Audio|GetAmbientRotationSound)))
