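; Quest listener gameplay ability for the Season 22 Signal Booster, with helpers for detecting when a spinning radar dish faces the tree.
; Helpers cache dish transforms, pick spin speed by completed objective count, and apply or remove a valid-tag GameplayEffect.
; The facing test derives current and predicted vectors from yaw and dot-products them against the tree direction.
; Activate and end events only call their parents, and RadarUpdateLoop is empty, leaving helpers unconnected.
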(event Ability|EventActivateAbilityFromEvent
  (params (EventData :type Gameplay%20Event%20Data%20Structure%20%28by%20ref%29))
  ;[0 0 0] ActivateAbilityFromEvent - when valid player approaches the dish, init starting radar vectors and set a timer to begin calculations
  (|CallParentFunction EventData))

(event Custom|RadarUpdateLoop)

(event Ability|EventOnEndAbility
  (params (bWasCancelled :type Boolean))
  ;[0 0 0] OnEndAbility - clear timer
  (|CallParentFunction bWasCancelled))

(fn ApplyOrRemoveValidGE
  (params (bShouldApply :type Boolean))
  (if bShouldApply
    (bind _returnvalue (GameplayEffects|ApplyGameplayEffectToSelf :self (Ability|GetFortAbilitySystemComponent (Variables|GAVars|Data|GetOwningPlayerPawn)) :GameplayEffectClass (LocalVariables|GetSignalBoosterGrantValidTagGE) :Level 0.0))
    (Variables|Default|SetValidTagGEHandle _returnvalue)
    (return)
    (else
      (bind _returnvalue_2 (Ability|GameplayEffect|GetActiveGameplayEffectStackCount (Variables|Default|GetValidTagGEHandle)))
      (if (== _returnvalue_2 1)
        (Ability|RemoveGameplayEffectFromOwnerWithHandle (Variables|Default|GetValidTagGEHandle))
        (else
          _))
      (return))))

(fn InitRadarVectors
  (params (SignalBooster :type BP%20S22%20Signal%20Booster%20Old%20Object%20Reference))
  (bind _sm_quest_preinteraction (Variables|/Game/Athena/Items/QuestParents/PropQuestActor/BP_Athena_PropQuestActor_Parent.BP_Athena_PropQuestActor_Parent_C|GetSM_Quest_PreInteraction SignalBooster))
  ;[0 0 0] Get the initial world position, direction, and rotation of the radar dish
  (Variables|Default|SetRadarWorldLocation (Transformation|GetWorldLocation _sm_quest_preinteraction))
  ;;[0.1406 0.1406 0.1406] Because of the way the mesh is set up, the Right Vector is actually the way the dish faces forward
  (bind _output_get (Variables|Default|SetRadarDish (Components|GetChildComponent _sm_quest_preinteraction 1)))
  (bind _output_get_2 (Variables|Default|SetRadarDirectionalVectorPredicted (Transformation|GetRightVector _output_get)))
  (Variables|Default|SetRadarDirectionalVectorCurrent _output_get_2)
  (Variables|Default|SetDishRotationPredicted (Variables|/Script/Engine.SceneComponent|GetRelativeRotation _output_get))
  (return))

(fn IsRadarFacingTheTree
  (bind (_dishrotationpredicted_roll _dishrotationpredicted_pitch _dishrotationpredicted_yaw) (Math|Rotator|BreakRotator (Variables|GetDishRotationPredicted_Roll)))
  (bind (_dishrotationcurrent_roll _dishrotationcurrent_pitch _dishrotationcurrent_yaw) (Math|Rotator|BreakRotator (Variables|GetDishRotationCurrent_Roll)))
  ;[0 0 0] Do math to calculate what the predicted forward-facing vector will be, and if it will face the tree on the next timer execution
  (Utilities|FlowControl|Sequence
    (:then_0
      ;;[0.1406 0.1406 0.1406] Because of the way the mesh is set up, the Right Vector is actually the way the dish faces forward. Thus, we calculate for the right vector given the rotation:
      (Variables|SetRadarDirectionalVectorPredicted_X (Math|Vector|MakeVector (* (Math|Trig|Sin(Degrees) _dishrotationpredicted_yaw) -1.0) (Math|Trig|Cos(Degrees) _dishrotationpredicted_yaw))))
    (:then_1
      ;;[0.1458 0.1458 0.1458] Check against the current and predicted rotational values
      (Variables|SetRadarDirectionalVectorCurrent_X (Math|Vector|MakeVector (* (Math|Trig|Sin(Degrees) _dishrotationcurrent_yaw) -1.0) (Math|Trig|Cos(Degrees) _dishrotationcurrent_yaw))))
    (:then_2
      ;;[0.1406 0.1406 0.1406] Calculate the dot product for the vector connecting and pointing from radar to tree and the radar's directional vector. Do this for both the current and predicted vector
      ;;[0.1406 0.1406 0.1406] Product returns: Facing tree exact: 1.0; somewhat facing tree: 0.0-1.0; facing away from tree: negative
      (if (or (>= (Math|Vector|DotProduct (Math|Vector|Normalize (- (Variables|Default|GetTreeWorldLocation) (Variables|Default|GetRadarWorldLocation))) (Variables|Default|GetRadarDirectionalVectorCurrent)) (LocalVariables|GetDotProductTolerance)) (>= (Math|Vector|DotProduct (Math|Vector|Normalize (- (Variables|Default|GetTreeWorldLocation) (Variables|Default|GetRadarWorldLocation))) (Variables|Default|GetRadarDirectionalVectorPredicted)) (LocalVariables|GetDotProductTolerance)))
        (return true)
        (else
          (return false))))))

(fn SetRadarSpeedforCurrentObjective
  (bind _signalboosterquest (Variables|Default|GetSignalBoosterQuest))
  (bind _owningplayerquestmanager (Variables|GAVars|Data|GetOwningPlayerQuestManager))
  (bind _returnvalue (Quests|GetQuestwithDefinition _owningplayerquestmanager _signalboosterquest))
  (bind _signalboosterspeedsettings (Variables|Default|GetSignalBoosterSpeedSettings))
  ;[0 0 0] Depending on which objective the player is on, set the radar dish spin speed
  ;;[0.1094 0.1094 0.1094] Spin Settings will likely need to be manually set for Visual Test Mode iteration
  (LocalVariables|SetNumObjectivesCompleted 0)
  (Utilities|IsValid _returnvalue
    (:Is%20Valid
      ;;[0.1098 0.1098 0.1098] Check how many objectives have been completed, and set the turn speed (in degrees) based on enum -> float map values
      (for _array_element (Variables|/Script/FortniteGame.FortQuestItem|GetObjectives _returnvalue)
        (if (Quest|HasCompleted _array_element)
          (LocalVariables|SetNumObjectivesCompleted (+ (LocalVariables|GetNumObjectivesCompleted) 1))))
      (switch int (LocalVariables|GetNumObjectivesCompleted)
        (:Default
          (Variables|Default|SetDegreeTurn (Utilities|Map|Find _signalboosterspeedsettings "NewEnumerator0")))
        (:0
          (Variables|Default|SetDegreeTurn (Utilities|Map|Find _signalboosterspeedsettings "NewEnumerator1")))
        (:1
          (Variables|Default|SetDegreeTurn (Utilities|Map|Find _signalboosterspeedsettings "NewEnumerator2")))
        (:2
          (Variables|Default|SetDegreeTurn (Utilities|Map|Find _signalboosterspeedsettings "NewEnumerator3")))))
    (:Is%20Not%20Valid
      _))
  (Development|PrintString (Utilities|String|Append "Speed: " (Utilities|String|ToString(Float) (Variables|Default|GetDegreeTurn))) true true "(R=1.000000,G=0.000000,B=0.172688,A=1.000000)")
  (return))
